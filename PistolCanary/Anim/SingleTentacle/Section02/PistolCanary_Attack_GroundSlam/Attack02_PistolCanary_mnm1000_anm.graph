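# Level Sequence for shot mnm1000 of the PistolCanary Section02 tentacle ground slam attack, an animation cinematic.
# The spawned tentacle creature is driven by a Control Rig track with a full-weight section and two partial-weight sections.
# Six superhero NPCs switch visible at the frame where the single camera cut begins; their skeletal animation tracks hold no sections.
# A CineCamera with unkeyed transform and lens values provides a static view.
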LevelSequence /PistolCanary/Anim/SingleTentacle/Section02/PistolCanary_Attack_GroundSlam/Attack02_PistolCanary_mnm1000_anm
bindings [
 PistolCanary_CinematicNPC_Superhero2 {
  id=#1
  possessed_class=PistolCanary_CinematicNPC_Superhero2_C
  tracks [
   {type=MovieSceneSkeletalAnimationTrack, sections []}
   {
    type=MovieSceneVisibilityTrack
    sections [{type=MovieSceneVisibilitySection, channels [None {type=MovieSceneScriptingBoolChannel, default=false, keys [[1799.5,false],[1800,true]]}]}]
   }
  ]
 }
 Tentacle_PistolCanary_Creature {
  id=#2
  possessed_class=FortSkeletalMeshActor
  tracks [
   {type=MovieSceneSpawnTrack, sections [{type=MovieSceneSpawnSection, channels [None {type=MovieSceneScriptingBoolChannel, default=true}]}]}
   {
    type=MovieScene3DTransformTrack
    sections [{
      type=MovieScene3DTransformSection
      channels [
       Location.X {type=MovieSceneScriptingDoubleChannel, default=0}
       Location.Y {type=MovieSceneScriptingDoubleChannel, default=0}
       Location.Z {type=MovieSceneScriptingDoubleChannel, default=0}
       Rotation.X {type=MovieSceneScriptingDoubleChannel, default=0}
       Rotation.Y {type=MovieSceneScriptingDoubleChannel, default=0}
       Rotation.Z {type=MovieSceneScriptingDoubleChannel, default=0}
       Scale.X {type=MovieSceneScriptingDoubleChannel, default=1}
       Scale.Y {type=MovieSceneScriptingDoubleChannel, default=1}
       Scale.Z {type=MovieSceneScriptingDoubleChannel, default=1}
      ]
     }]
   }
   {type=MovieSceneSkeletalAnimationTrack, sections []}
   {
    type=MovieSceneControlRigParameterTrack
    display_name=Tentacle_PistolCanary_Creature_CtrlRig
    sections [
     {
      type=MovieSceneControlRigParameterSection
      channels [
       control rig options.Location.X {type=MovieSceneScriptingFloatChannel, default=0}
       control rig options.Location.Y {type=MovieSceneScriptingFloatChannel, default=0}
       control rig options.Location.Z {type=MovieSceneScriptingFloatChannel, default=0}
       control rig options.Rotation.X {type=MovieSceneScriptingFloatChannel, default=0}
       control rig options.Rotation.Y {type=MovieSceneScriptingFloatChannel, default=0}
       control rig options.Rotation.Z {type=MovieSceneScriptingFloatChannel, default=0}
       control rig options.Scale.X {type=MovieSceneScriptingFloatChannel, default=1}
       control rig options.Scale.Y {type=MovieSceneScriptingFloatChannel, default=1}
       control rig options.Scale.Z {type=MovieSceneScriptingFloatChannel, default=1}
       Weight {type=MovieSceneScriptingFloatChannel, default=1}
      ]
     }
     {type=MovieSceneControlRigParameterSection, channels [Weight {type=MovieSceneScriptingFloatChannel, default=0.3}]}
     {type=MovieSceneControlRigParameterSection, channels [Weight {type=MovieSceneScriptingFloatChannel, default=0.5}]}
    ]
   }
  ]
 }
 PistolCanary_CinematicNPC_Superhero3 {
  id=#3
  possessed_class=PistolCanary_CinematicNPC_Superhero3_C
  tracks [
   {type=MovieSceneSkeletalAnimationTrack, sections []}
   {
    type=MovieSceneVisibilityTrack
    sections [{type=MovieSceneVisibilitySection, channels [None {type=MovieSceneScriptingBoolChannel, default=false, keys [[1799.5,false],[1800,true]]}]}]
   }
  ]
 }
 CameraComponent {
  id=#4
  parent_id=#5
  possessed_class=CineCameraComponent
  tracks [
   {
    type=MovieSceneFloatTrack
    display_name=Current Focal Length
    sections [{type=MovieSceneFloatSection, channels [None {type=MovieSceneScriptingFloatChannel, default=35}]}]
   }
   {
    type=MovieSceneFloatTrack
    display_name="Manual Focus Distance (Focus Settings)"
    sections [{type=MovieSceneFloatSection, channels [None {type=MovieSceneScriptingFloatChannel, default=100000}]}]
   }
   {
    type=MovieSceneFloatTrack
    display_name=Current Aperture
    sections [{type=MovieSceneFloatSection, channels [None {type=MovieSceneScriptingFloatChannel, default=2.8}]}]
   }
  ]
 }
 PistolCanary_CinematicNPC_Superhero1 {
  id=#6
  possessed_class=PistolCanary_CinematicNPC_Superhero1_C
  tracks [
   {type=MovieSceneSkeletalAnimationTrack, sections []}
   {
    type=MovieSceneVisibilityTrack
    sections [{type=MovieSceneVisibilitySection, channels [None {type=MovieSceneScriptingBoolChannel, default=false, keys [[1799.5,false],[1800,true]]}]}]
   }
  ]
 }
 CineCameraActor {
  id=#5
  possessed_class=CineCameraActor
  tracks [{
    type=MovieScene3DTransformTrack
    sections [{
      type=MovieScene3DTransformSection
      channels [
       Location.X {type=MovieSceneScriptingDoubleChannel, default=43812.403}
       Location.Y {type=MovieSceneScriptingDoubleChannel, default=108733.2413}
       Location.Z {type=MovieSceneScriptingDoubleChannel, default=4792.1439}
       Rotation.X {type=MovieSceneScriptingDoubleChannel, default=-0.0424}
       Rotation.Y {type=MovieSceneScriptingDoubleChannel, default=7.6022}
       Rotation.Z {type=MovieSceneScriptingDoubleChannel, default=-129.4057}
       Scale.X {type=MovieSceneScriptingDoubleChannel, default=1}
       Scale.Y {type=MovieSceneScriptingDoubleChannel, default=1}
       Scale.Z {type=MovieSceneScriptingDoubleChannel, default=1}
      ]
     }]
   }]
 }
 PistolCanary_CinematicNPC_Superhero6 {
  id=#7
  possessed_class=PistolCanary_CinematicNPC_Superhero6_C
  tracks [
   {type=MovieSceneSkeletalAnimationTrack, sections []}
   {
    type=MovieSceneVisibilityTrack
    sections [{type=MovieSceneVisibilitySection, channels [None {type=MovieSceneScriptingBoolChannel, default=false, keys [[1799.5,false],[1800,true]]}]}]
   }
  ]
 }
 PistolCanary_CinematicNPC_Superhero4 {
  id=#8
  possessed_class=PistolCanary_CinematicNPC_Superhero4_C
  tracks [
   {type=MovieSceneSkeletalAnimationTrack, sections []}
   {
    type=MovieSceneVisibilityTrack
    sections [{type=MovieSceneVisibilitySection, channels [None {type=MovieSceneScriptingBoolChannel, default=false, keys [[1799.5,false],[1800,true]]}]}]
   }
  ]
 }
 PistolCanary_CinematicNPC_Superhero5 {
  id=#9
  possessed_class=PistolCanary_CinematicNPC_Superhero5_C
  tracks [
   {type=MovieSceneSkeletalAnimationTrack, sections []}
   {
    type=MovieSceneVisibilityTrack
    sections [{type=MovieSceneVisibilitySection, channels [None {type=MovieSceneScriptingBoolChannel, default=false, keys [[1799.5,false],[1800,true]]}]}]
   }
  ]
 }
]
display_rate {denominator=1, numerator=30}
tick_resolution {denominator=1, numerator=24000}
tracks [{type=MovieSceneCameraCutTrack, sections [{type=MovieSceneCameraCutSection, end_frame=6042, start_frame=1800}]}]
